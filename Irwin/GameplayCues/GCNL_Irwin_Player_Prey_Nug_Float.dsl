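; Looping gameplay cue that plays the float sound for the Nug prey player in Irwin.
; On loop start, with a valid target, it fades out any stored float sound, then spawns the Nug float loop attached to the target's root component and stores the audio component.
; On removal, it fades out the stored sound over a slightly longer time.
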(event EventOnLoopingStart(EitherVFX-preferthisversion)
  (params (MyTarget :type Actor%20Object%20Reference) (Parameters :type Gameplay%20Cue%20Parameters%20Structure%20%28by%20ref%29) (ParticleComponents :type Array%20of%20FXSystem%20Component%20Object%20References) (AudioComponents :type Array%20of%20Audio%20Component%20Object%20References))
  (Utilities|IsValid MyTarget
    (:Is%20Valid
      (bind _floatsound (Variables|Default|GetFloatSound)
        (:then
          (Audio|Components|Audio|FadeOut _floatsound 0.1 0.0 "Linear" :!class /Script/Engine.AudioComponent))
        (:else
          _))
      (bind _returnvalue (Audio|SpawnSoundAttached "/Irwin/Sounds/Nug/Nug_Float_Loop_Cue.Nug_Float_Loop_Cue" (Variables|/Script/Engine.Actor|GetRootComponent MyTarget)))
      (Variables|Default|SetFloatSound _returnvalue))))

(event EventOnRemoval(EitherVFX-preferthisversion)
  (params (MyTarget :type Actor%20Object%20Reference) (Parameters :type Gameplay%20Cue%20Parameters%20Structure%20%28by%20ref%29) (ParticleComponents :type Array%20of%20FXSystem%20Component%20Object%20References) (AudioComponents :type Array%20of%20Audio%20Component%20Object%20References) (BurstCameraShakeInstance :type Camera%20Shake%20Base%20Object%20Reference) (BurstDecalInstance :type Decal%20Actor%20Object%20Reference))
  (bind _floatsound (Variables|Default|GetFloatSound)
    (:then
      (Audio|Components|Audio|FadeOut _floatsound 0.25 0.0 "Linear" :!class /Script/Engine.AudioComponent))
    (:else)))

(event EventBeginPlay)

(event Collision|EventActorBeginOverlap
  (params (OtherActor :type Actor%20Object%20Reference)))

(event EventTick
  (params (DeltaSeconds :type Float%20%28single-precision%29)))

(fn ConstructionScript)
