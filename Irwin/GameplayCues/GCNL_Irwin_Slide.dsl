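; Looping gameplay cue that plays a slide sound on the target Fort pawn while the cue is active.
; On looping start it spawns the sound attached to the pawn's root component, quieter when the pawn is friendly to the local player, and stores the audio component.
; On removal it fades that stored component out over a short linear fade.
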(event EventOnLoopingStart(EitherVFX-preferthisversion)
  (params (MyTarget :type Actor%20Object%20Reference) (Parameters :type Gameplay%20Cue%20Parameters%20Structure%20%28by%20ref%29) (ParticleComponents :type Array%20of%20FXSystem%20Component%20Object%20References) (AudioComponents :type Array%20of%20Audio%20Component%20Object%20References))
  (bind _asfort_pawn (Utilities|Casting|CastToFortPawn MyTarget))
  (bind _returnvalue (Audio|SpawnSoundAttached :Sound (Variables|Default|GetSlideSound) :AttachToComponent (Variables|/Script/Engine.Actor|GetRootComponent _asfort_pawn) :AttachPointName "None" :LocationType "KeepRelativeOffset" :bStopWhenAttachedToDestroyed false :VolumeMultiplier (Utilities|Select :Index (Team|GetTeamAffiliationforActortoLocalPlayer _asfort_pawn) :Friendly 0.4 :Neutral 1.0 :Hostile 1.0) :PitchMultiplier 1.0 :StartTime 0.0 :bAutoDestroy true))
  (Variables|Default|SetSlidingSoundComponent _returnvalue))

(event EventOnRemoval(EitherVFX-preferthisversion)
  (params (MyTarget :type Actor%20Object%20Reference) (Parameters :type Gameplay%20Cue%20Parameters%20Structure%20%28by%20ref%29) (ParticleComponents :type Array%20of%20FXSystem%20Component%20Object%20References) (AudioComponents :type Array%20of%20Audio%20Component%20Object%20References) (BurstCameraShakeInstance :type Camera%20Shake%20Base%20Object%20Reference) (BurstDecalInstance :type Decal%20Actor%20Object%20Reference))
  (Audio|Components|Audio|FadeOut (Variables|Default|GetSlidingSoundComponent) 0.5 0.0 "Linear" :!class /Script/Engine.AudioComponent))

(fn ConstructionScript)
