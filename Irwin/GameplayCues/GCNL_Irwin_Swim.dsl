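; Looping gameplay cue that drives water-aware swimming VFX for Irwin AI pawns.
; On loop start it caches the target AI pawn, its AI movement component and a spawned particle component.
; Each tick, when the pawn has a water body or is surface swimming, it feeds body index, immersion depth and surface plane location and normal into the particle parameters.
; Debug prints report those values and failed casts or invalid references.
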(event EventOnLoopingStart(EitherVFX-preferthisversion)
  (params (MyTarget :type Actor%20Object%20Reference) (Parameters :type Gameplay%20Cue%20Parameters%20Structure%20%28by%20ref%29) (ParticleComponents :type Array%20of%20FXSystem%20Component%20Object%20References) (AudioComponents :type Array%20of%20Audio%20Component%20Object%20References))
  (bind _asfort_aipawn (Utilities|Casting|CastToFortAIPawn MyTarget)
    (:then
      (Variables|Default|SetIrwin _asfort_aipawn)
      (bind _asfort_movement_comp_aichar (Utilities|Casting|CastToFortMovementComp_AIChar (Variables|/Script/Engine.Character|GetCharacterMovement _asfort_aipawn))
        (:then
          (Variables|Default|SetIrwin%20Movement _asfort_movement_comp_aichar)
          (Variables|Default|SetSwim%20VFX (Utilities|Array|Get(acopy) ParticleComponents))
          (Development|PrintString 0 "Irwin Swimming GC Active"))
        (:CastFailed
          (Development|PrintString 0 "GCNL_Swim - Movement Comp CAST FAILED"))))
    (:CastFailed
      (Development|PrintString 0 "GCNL_Swim - AI Pawn CAST FAILED"))))

(event EventTick
  (params (DeltaSeconds :type Float%20%28single-precision%29))
  (bind _irwin (Variables|Default|GetIrwin))
  (bind _returnvalue (WaterBody|GetCurrentWaterBody _irwin))
  (bind _returnvalue_2 (WaterBody|GetWaterImmersionDepth _irwin))
  (bind _irwin_movement (Variables|Default|GetIrwin%20Movement))
  (bind _returnvalue_3 (Swimming|GetWaterSurfaceLocation _irwin_movement))
  (bind _returnvalue_4 (Swimming|GetWaterSurfaceNormal _irwin_movement))
  (bind _swim_vfx (Variables|Default|GetSwim%20VFX))
  (bind _waterbodyindex (Variables|/Script/Water.WaterBody|GetWaterBodyIndex _returnvalue))
  (if (Utilities|IsValid _irwin :!class /Script/Engine.KismetSystemLibrary)
    (if (or (Utilities|IsValid _returnvalue :!class /Script/Engine.KismetSystemLibrary) (WaterBody|IsSurfaceSwimming _irwin))
      (Utilities|IsValid _swim_vfx
        (:Is%20Valid
          (Effects|Components|ParticleSystem|SetIntParameter _swim_vfx "WaterBodyIndex" _waterbodyindex)
          (Effects|Components|ParticleSystem|SetFloatParameter _swim_vfx "WaterDepth" _returnvalue_2)
          (Utilities|IsValid _irwin_movement
            (:Is%20Valid
              (Effects|Components|ParticleSystem|SetVectorParameter _swim_vfx "WaterPlaneLocation" _returnvalue_3)
              (Effects|Components|ParticleSystem|SetVectorParameter _swim_vfx "WaterPlaneNormal" _returnvalue_4)
              (Development|PrintString self (Utilities|String|Append "Water Body Index - " (Utilities|String|ToString(Integer) _waterbodyindex) "
" "Water Depth - " (Utilities|String|ToString(Float) _returnvalue_2) "
" "Water Plane Location - " (Utilities|String|ToString(Vector) _returnvalue_3) "
" "Water Plane Normal - " (Utilities|String|ToString(Vector) _returnvalue_4) "
") false true "(R=0.000000,G=0.660000,B=1.000000,A=1.000000)" 0.05))
            (:Is%20Not%20Valid
              (Development|PrintString self (Utilities|String|Append "IrwinSwimVFX_Irwin Movement_NOT VALID - " (Utilities|GetDisplayName _irwin_movement)) false))))
        (:Is%20Not%20Valid
          (Development|PrintString self (Utilities|String|Append "IrwinSwimVFX_NOT VALID - " (Utilities|GetDisplayName _swim_vfx)) false))))))

(fn ConstructionScript)
