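; Gameplay cue that plays the rez-out dissolve effect on an NPC pawn, with variants for wildlife types.
; On burst it swaps in the dissolve material, starts a timeline, spawns mesh-bound Niagara and hides child FX components.
; A pre-sequence event derives fauna type from gameplay tags and chromed state from the base material, feeding both to Niagara.
; A vertical trace adds a splash scaled by mesh bounds and fed water-surface data.
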(event EventOnBurst(EitherVFX-preferthisversion)
  (params (MyTarget :type Actor%20Object%20Reference) (Parameters :type Gameplay%20Cue%20Parameters%20Structure%20%28by%20ref%29) (ParticleComponents :type Array%20of%20FXSystem%20Component%20Object%20References) (AudioComponents :type Array%20of%20Audio%20Component%20Object%20References) (BurstCameraShakeInstance :type Camera%20Shake%20Base%20Object%20Reference) (BurstDecalInstance :type Decal%20Actor%20Object%20Reference))
  (bind _irwin (Variables|DONOTEDIT|GetIrwin))
  (bind _mesh (Variables|/Script/Engine.Character|GetMesh _irwin))
  (bind _asnpc_pawn_parent (Utilities|Casting|CastToNPC_Pawn_Parent MyTarget))
  (Variables|DONOTEDIT|SetIrwin _asnpc_pawn_parent)
  (CallFunction|Pre-SequenceChange)
  (CallFunction|SetIrwinMaterials)
  (CallFunction|WaterTrace)
  (CallFunction|SetTimelinePlayrates)
  (Components|Timeline|PlayfromStart (Variables|GCN_Irwin_RezOut_Simple|GetTFX_ResOutChracterMesh))
  (CallFunction|SpawnDissolveVFX)
  (CallFunction|SpawnWaterEliminationFX)
  (Utilities|IsValid _mesh
    (:Is%20Valid
      (for _array_element (Components|GetChildrenComponents _mesh true)
        (bind _asfxsystem_component (Utilities|Casting|CastToFXSystemComponent _array_element))
        (Development|SetHiddeninGame _asfxsystem_component true true)
        (if (Variables|DEBUG|GetDEBUG_FXSYSTEMCOMPONENTS)
          (Development|PrintString 0 (Utilities|GetDisplayName _array_element)))))))

(event Custom|Pre-Sequence%20Change
  (bind _irwin (Variables|DONOTEDIT|GetIrwin))
  (bind _gameplaytags (Variables|/Script/FortniteGame.FortPawn|GetGameplayTags _irwin))
  ;[1 1 1] Alteration Events to alter Sequence without needing to re-write large sections
  (if (GameplayTags|HasTag _gameplaytags "(TagName=\"Pawn.Athena.NPC.Wildlife.Predator.Grandma\")")
    (Variables|Default|SetClass%20Test 0 :!label _label)
    (elif (GameplayTags|HasTag _gameplaytags "(TagName=\"Pawn.Athena.NPC.Wildlife.Prey.Burt\")")
      (Variables|Default|SetClass%20Test 1)
      (elif (GameplayTags|HasTag _gameplaytags "(TagName=\"Pawn.Athena.NPC.Wildlife.Prey.Nug\")")
        (Variables|Default|SetClass%20Test 2)
        (elif (GameplayTags|HasTag _gameplaytags "(TagName=\"Pawn.Athena.NPC.Wildlife.Simple.Smackie\")")
          (Variables|Default|SetClass%20Test 3)
          (else
            (goto _label))))))
  (Development|PrintString 0 (Utilities|String|ToString(Integer) (Variables|Default|GetClass%20Test)))
  (bind _returnvalue (Rendering|Material|GetBaseMaterial (Rendering|Material|GetMaterial (Variables|/Script/Engine.Character|GetMesh (Variables|DONOTEDIT|GetIrwin)) 0)))
  (Variables|Default|SetWasChromed (== _returnvalue "/Game/TechArt/Seasonal/LackLuster/Materials/M_LackLuster_AI_Parent.M_LackLuster_AI_Parent")))

(event Custom|Set%20Additional%20Niagara%20Parameters%20on%20Spawned%20FX
  (params (Spawned%20Death%20FX :type Niagara%20Particle%20System%20Component%20Object%20Reference))
  ;[1 1 1] Alteration Events to alter Sequence without needing to re-write large sections
  (Effects|Components|ParticleSystem|SetBoolParameter Spawned%20Death%20FX "WasChromed" (Variables|Default|GetWasChromed))
  (Effects|Components|ParticleSystem|SetIntParameter Spawned%20Death%20FX "FaunaSelection" (Variables|Default|GetClass%20Test)))

(fn Clean-UpTeleportationLight)

(fn SetIrwinMaterials
  (bind _irwin (Variables|DONOTEDIT|GetIrwin))
  (bind _mesh (Variables|/Script/Engine.Character|GetMesh _irwin))
  (Utilities|IsValid _mesh
    (:Is%20Valid
      (for _index (range (+ (Utilities|Array|Length (Rendering|Material|GetMaterials _mesh)) 1))
        (Rendering|Material|SetMaterial _mesh _index (Variables|DISSOLVE|GetMat%20Chracter%20Dissolve)))
      (return))))

(fn SetTimelinePlayrates
  (Components|Timeline|SetPlayRate (Variables|GCN_Irwin_RezOut_Simple|GetTFX_ResOutChracterMesh) (Variables|DISSOLVE|GetDissolve%20Timeline%20Playrate)))

(fn SocketMath
  (bind _irwin (Variables|DONOTEDIT|GetIrwin)
    (:then
      (Utilities|IsValid (Customization|GetSkeletalMeshforPartType _irwin "Body" :!class /Script/FortniteGame.FortAIPawn)
        (:Is%20Valid
          (return (Math|Vector|VectorLength (- (+ (Transformation|GetSocketLocation (Customization|GetSkeletalMeshforPartType _irwin "Body" :!class /Script/FortniteGame.FortAIPawn) (Variables|DISSOLVE|GetSocket%20Mesh%20Bottom)) (* (Math|Vector|Normalize (- (Transformation|GetSocketLocation (Customization|GetSkeletalMeshforPartType _irwin "Body" :!class /Script/FortniteGame.FortAIPawn) (Variables|DISSOLVE|GetSocket%20Mesh%20Top)) (Transformation|GetSocketLocation (Customization|GetSkeletalMeshforPartType _irwin "Body" :!class /Script/FortniteGame.FortAIPawn) (Variables|DISSOLVE|GetSocket%20Mesh%20Bottom)))) -1e+01)) (+ (Transformation|GetSocketLocation (Customization|GetSkeletalMeshforPartType _irwin "Body" :!class /Script/FortniteGame.FortAIPawn) (Variables|DISSOLVE|GetSocket%20Mesh%20Top)) (* (Math|Vector|Normalize (- (Transformation|GetSocketLocation (Customization|GetSkeletalMeshforPartType _irwin "Body" :!class /Script/FortniteGame.FortAIPawn) (Variables|DISSOLVE|GetSocket%20Mesh%20Top)) (Transformation|GetSocketLocation (Customization|GetSkeletalMeshforPartType _irwin "Body" :!class /Script/FortniteGame.FortAIPawn) (Variables|DISSOLVE|GetSocket%20Mesh%20Bottom)))) 4e+01)))) (+ (Transformation|GetSocketLocation (Customization|GetSkeletalMeshforPartType _irwin "Body" :!class /Script/FortniteGame.FortAIPawn) (Variables|DISSOLVE|GetSocket%20Mesh%20Bottom)) (* (Math|Vector|Normalize (- (Transformation|GetSocketLocation (Customization|GetSkeletalMeshforPartType _irwin "Body" :!class /Script/FortniteGame.FortAIPawn) (Variables|DISSOLVE|GetSocket%20Mesh%20Top)) (Transformation|GetSocketLocation (Customization|GetSkeletalMeshforPartType _irwin "Body" :!class /Script/FortniteGame.FortAIPawn) (Variables|DISSOLVE|GetSocket%20Mesh%20Bottom)))) -1e+01)) (* (Math|Vector|Normalize (- (Transformation|GetSocketLocation (Customization|GetSkeletalMeshforPartType _irwin "Body" :!class /Script/FortniteGame.FortAIPawn) (Variables|DISSOLVE|GetSocket%20Mesh%20Top)) (Transformation|GetSocketLocation (Customization|GetSkeletalMeshforPartType _irwin "Body" :!class /Script/FortniteGame.FortAIPawn) (Variables|DISSOLVE|GetSocket%20Mesh%20Bottom)))) -1.0)))
        (:Is%20Not%20Valid
          _)))
    (:else
      _)) ; null check performed by caller, but if this ever fails we can use the same defaults as having an invalid mesh
  (return 0.0 "0, 0, 0" "0, 0, 0"))

(fn SpawnDissolveVFX
  (bind _irwin (Variables|DONOTEDIT|GetIrwin))
  (bind _mesh (Variables|/Script/Engine.Character|GetMesh _irwin))
  (if (Variables|DEBUG|GetDEBUG_REMOVESPAWNVFX)
    (Development|PrintString 0 "DEBUG - Removed Spawned VFX")
    (return)
    (else
      (Utilities|IsValid _mesh
        (:Is%20Valid
          (bind _returnvalue (Niagara|SpawnSystematLocation 0 (Variables|DISSOLVE|GetDissolve%20Visual%20Effect) (Transformation|GetSocketLocation _mesh (Variables|DISSOLVE|GetDissolve%20VFX%20Spawn%20Point%20Name))))
          (bind _output_get (Variables|DONOTEDIT|SetSpawned%20Death%20VFX _returnvalue))
          (Niagara|SetNiagaraSkeletalMeshComponent _output_get "SkeletalMesh" (Variables|/Script/Engine.Character|GetMesh (Variables|DONOTEDIT|GetIrwin)))
          (Niagara|SetNiagaraVariable(Vector3) _output_get "PointPosition" "0, 0,250.000000")
          (CallFunction|SetAdditionalNiagaraParametersOnSpawnedFX (Variables|DONOTEDIT|GetSpawned%20Death%20VFX))
          (return))))))

(fn SpawnTeleportationLight
  (bind _irwin (Variables|DONOTEDIT|GetIrwin)
    (:then
      (if (Variables|DEBUG|GetDEBUG_REMOVESPAWNEDPOINTLIGHT)
        _
        (elif (Fortnite|IsMobilePlatform)
          (Development|PrintString (Variables|DONOTEDIT|GetIrwin) "Rez Out - MOBILE PLATFORM - No Point LIght for you.")
          (else
            (AddComponent|NODE_AddPointLightComponent-0 :RelativeTransform (Math|Conversions|ToTransform(Vector) (+ (Transformation|GetSocketLocation (Customization|GetSkeletalMeshforPartType _irwin "Body" :!class /Script/FortniteGame.FortAIPawn) "root") "0, 0, 0")))))))
    (:else)))

(fn SpawnWaterEliminationFX
  (bind _water_hit_information (Variables|WaterDissolve|GetWater%20Hit%20Information))
  (bind (_bblockinghit _binitialoverlap _time _distance _location _impactpoint _normal _impactnormal _physmat _hitactor) (Collision|BreakHitResult _water_hit_information))
  (if (Variables|WaterDissolve|GetWater%20Hit)
    (bind _asfort_water_body_actor (Utilities|Casting|CastToFortWaterBodyActor _hitactor))
    (bind _outwatersurfaceinfo (WaterBody|GetWaterSurfaceInfoStruct _asfort_water_body_actor _location true false false))
    (bind _returnvalue (Niagara|SpawnSystematLocation (Variables|DONOTEDIT|GetIrwin) (Variables|WaterDissolve|GetWater%20Elimination%20Splash) (Utilities|Struct|BreakWaterSurfaceInfo _outwatersurfaceinfo) "90.000000, 0, 0"))
    (Effects|Components|ParticleSystem|SetVectorParameter _returnvalue "WaterPlaneLocation" (Utilities|Struct|BreakWaterSurfaceInfo _outwatersurfaceinfo))
    (Effects|Components|ParticleSystem|SetVectorParameter _returnvalue "WaterPlaneNormal" (.WaterPlaneNormal (Utilities|Struct|BreakWaterSurfaceInfo _outwatersurfaceinfo)))
    (Effects|Components|ParticleSystem|SetFloatParameter _returnvalue "WaterDepth" (.WaterDepth (Utilities|Struct|BreakWaterSurfaceInfo _outwatersurfaceinfo)))
    (Effects|Components|ParticleSystem|SetFloatParameter _returnvalue "WaterBodyIndex" (Math|Conversions|ToFloat(Integer) (.WaterBodyIdx (Utilities|Struct|BreakWaterSurfaceInfo _outwatersurfaceinfo))))
    (Effects|Components|ParticleSystem|SetFloatParameter _returnvalue "BoundScale" (Variables|WaterDissolve|GetBounds%20Scalar))
    (return)
    (else
      (return))))

(fn ConstructionScript)

(fn WaterTrace
  (bind _irwin (Variables|DONOTEDIT|GetIrwin))
  (bind _mesh (Variables|/Script/Engine.Character|GetMesh _irwin))
  (bind (_origin _boxextent _sphereradius) (Collision|GetComponentBounds _mesh))
  (Utilities|IsValid _mesh
    (:Is%20Valid
      (Variables|WaterDissolve|SetBounds%20Scalar (/ _sphereradius 130.5))
      (bind (_outhit _returnvalue) (Collision|LineTraceForObjects 0 (+ (Transformation|GetActorLocation (Variables|DONOTEDIT|GetIrwin)) "0, 0,100.000000") (- (Transformation|GetActorLocation (Variables|DONOTEDIT|GetIrwin)) "0, 0,100.000000") (Utilities|Array|MakeArray "ObjectTypeQuery12")))
      (if _returnvalue
        (Variables|WaterDissolve|SetWater%20Hit%20Information _outhit))
      (Variables|WaterDissolve|SetWater%20Hit _returnvalue)
      (return))))
